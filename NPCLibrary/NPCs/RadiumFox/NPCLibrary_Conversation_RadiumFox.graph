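# Conversation database for the RadiumFox heavy specialist NPC, giving its greeting and top-level conversation branches.
# One entry point leads to a speech node with a default greeting and contextual lines for first-ever meetings, match conversation count, and players wearing the RadiumFox Flame outfit.
# First-meeting and outfit lines use named priority curves.
# The greeting then links out to shared NPC quests, basic services, and RadiumFox's own dialogue entry.
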ConversationDatabase /NPCLibrary/NPCs/RadiumFox/NPCLibrary_Conversation_RadiumFox
entry_points [{tag=Conversation.Entry.RadiumFox, destinations [#1]}]
nodes [
 ConversationEntryPointNode_7 {
  id=#1
  class=ConversationEntryPointNode
  outputs [#2]
  properties {entryTag {tagName=Conversation.Entry.RadiumFox}}
 }
 Speech_C_4 {
  id=#2
  class=Speech_C
  outputs [#3, #4, #5]
  properties {
   nodeName=""
   generalConfig {
    defaultMessage="Steel be with you!"
    contextualMessages [
     {
      message="Ad Victoriam, my friend! That means \"To victory,\" by the way... "
      contextRequirements [{requirement {class=FortControllerRequirement_IsFirstEverConversationWithNPC}}]
      priority {curve {rowName=Tandem.Speech.Priority.FirstEverInteraction}}
     }
     {
      message="Steel be with you!"
      contextRequirements [{requirement {
         class=FortControllerRequirement_CheckNPCMatchConversationCount
         properties {nPCUniqueIdTag {tagName=AISpawnerData.Type.Specialist.Heavy.RadiumFox}}
        }}]
      priority {curve {rowName=None}}
     }
     {
      message="Ad Victoriam!"
      contextRequirements [{requirement {
         class=FortControllerRequirement_CheckNPCMatchConversationCount
         properties {nPCUniqueIdTag {tagName=AISpawnerData.Type.Specialist.Heavy.RadiumFox}}
        }}]
      priority {curve {rowName=None}}
     }
     {
      message=You being careful out there?
      contextRequirements [{requirement {
         class=FortControllerRequirement_CheckNPCMatchConversationCount
         properties {nPCUniqueIdTag {tagName=AISpawnerData.Type.Specialist.Heavy.RadiumFox}}
        }}]
      priority {curve {rowName=None}}
     }
     {
      message="Remember, you need armor AND a weapon."
      contextRequirements [{requirement {
         class=FortControllerRequirement_CheckNPCMatchConversationCount
         properties {nPCUniqueIdTag {tagName=AISpawnerData.Type.Specialist.Heavy.RadiumFox}}
        }}]
      priority {curve {rowName=None}}
     }
     {
      message="Good to see you, friend of steel."
      contextRequirements [{requirement {
         class=FortControllerRequirement_CheckNPCMatchConversationCount
         properties {nPCUniqueIdTag {tagName=AISpawnerData.Type.Specialist.Heavy.RadiumFox}}
        }}]
      priority {curve {rowName=None}}
     }
     {
      message="Well look at you. Shiny!"
      contextRequirements [{requirement {class=FortControllerRequirement_HasCID, properties {allowedCIds [/BRCosmetics/Athena/Items/Cosmetics/Characters/Character_RadiumFox_Flame]}}}]
      priority {curve {rowName=Tandem.Speech.Priority.OutfitReaction}}
     }
     {
      message=So you're one of us now. Good.
      contextRequirements [{requirement {class=FortControllerRequirement_HasCID, properties {allowedCIds [&Character_RadiumFox_Flame]}}}]
      priority {curve {rowName=Tandem.Speech.Priority.OutfitReaction}}
     }
     {
      message="We are united in steel!"
      contextRequirements [{requirement {class=FortControllerRequirement_HasCID, properties {allowedCIds [&Character_RadiumFox_Flame]}}}]
      priority {curve {rowName=Tandem.Speech.Priority.OutfitReaction}}
     }
    ]
   }
  }
 }
 ConversationLinkNode_2 {id=#5, class=ConversationLinkNode, properties {remoteEntryTag {tagName=Conversation.Entry.RadiumFox.Dialogue}}}
 ConversationLinkNode_0 {id=#4, class=ConversationLinkNode, properties {remoteEntryTag {tagName=Conversation.Entry.BasicServices}}}
 ConversationLinkNode_1 {id=#3, class=ConversationLinkNode, properties {remoteEntryTag {tagName=Conversation.Entry.NPCLibrary.Quests}}}
]
